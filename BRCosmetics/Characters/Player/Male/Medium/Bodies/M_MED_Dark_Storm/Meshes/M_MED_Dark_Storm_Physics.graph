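# Physics asset for the M_MED_Dark_Storm male cosmetic body, covering four dangling dyn_ bone chains at front and back, left and right.
# Capsule chains hang from master bodies linked to the pelvis, with tight per-joint swing and twist limits.
# Pelvis, spine_01 and thighs supply body shapes.
# Free-angular constraints tie each front tip to the same-side back tip and also duplicate each front chain's last, limited joint.
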PhysicsAsset /BRCosmetics/Characters/Player/Male/Medium/Bodies/M_MED_Dark_Storm/Meshes/M_MED_Dark_Storm_Physics
bodies {
 dyn_back_l_01 [{shapeType=Capsule, length=10, radius=1.5, shapeName=dyn_back_l_01_capsule, center=<1 0 0>, rotation=<rot 90 0 0>}]
 dyn_back_l_02 [{shapeType=Capsule, length=10, radius=2.5, shapeName=dyn_back_l_02_capsule, center=<1 0 0>, rotation=<rot 90 0 0>}]
 dyn_back_l_03 [{shapeType=Capsule, length=10, radius=3, shapeName=dyn_back_l_03_capsule, center=<1 0 0>, rotation=<rot 90 0 0>}]
 dyn_back_l_04 [{shapeType=Capsule, length=9.75, radius=2.75, shapeName=dyn_back_l_04_capsule, center=<1 0 0>, rotation=<rot 90 0 0>}]
 dyn_back_l_05 [{shapeType=Capsule, length=9.5, radius=2.75, shapeName=dyn_back_l_05_capsule, center=<1 0 0>, rotation=<rot 90 0 0>}]
 dyn_back_l_06 [{shapeType=Capsule, length=7, radius=4, shapeName=dyn_back_l_05_capsule, center=<1 0 0>, rotation=<rot 90 0 0>}]
 dyn_back_r_01 [{shapeType=Capsule, length=10, radius=1.5, shapeName=dyn_back_l_01_capsule, center=<-1 0 0>, rotation=<rot 90 0 0>}]
 dyn_back_r_02 [{shapeType=Capsule, length=10, radius=2.75, shapeName=dyn_back_l_02_capsule, center=<-1 0 0>, rotation=<rot 90 0 0>}]
 dyn_back_r_03 [{shapeType=Capsule, length=10, radius=3, shapeName=dyn_back_l_03_capsule, center=<-1 0 0>, rotation=<rot 90 0 0>}]
 dyn_back_r_04 [{shapeType=Capsule, length=9.75, radius=2.75, shapeName=dyn_back_l_04_capsule, center=<-1 0 0>, rotation=<rot 90 0 0>}]
 dyn_back_r_05 [{shapeType=Capsule, length=9.5, radius=2.5, shapeName=dyn_back_l_05_capsule, center=<-1 0 0>, rotation=<rot 90 0 0>}]
 dyn_back_r_06 [{shapeType=Capsule, length=7, radius=3, shapeName=dyn_back_l_05_capsule, center=<-1 0 0>, rotation=<rot 90 0 0>}]
 dyn_front_l_01 [{shapeType=Capsule, length=3.25, radius=0.75, shapeName=dyn_front_l_01_capsule, center=<1 0 0>, rotation=<rot 90 0 0>}]
 dyn_front_l_02 [{shapeType=Capsule, length=10, radius=2.5, shapeName=dyn_front_l_02_capsule, center=<1 0 0>, rotation=<rot 90 0 0>}]
 dyn_front_l_03 [{shapeType=Capsule, length=9.75, radius=2.5, shapeName=dyn_front_l_03_capsule, center=<1 0 0>, rotation=<rot 90 0 0>}]
 dyn_front_l_04 [{shapeType=Capsule, length=9.5, radius=2.5, shapeName=dyn_front_l_04_capsule, center=<1 0 0>, rotation=<rot 90 0 0>}]
 dyn_front_l_05 [{shapeType=Capsule, length=7, radius=4, shapeName=dyn_back_l_05_capsule, center=<1 0 0>, rotation=<rot 90 0 0>}]
 dyn_front_r_01 [{shapeType=Capsule, length=3.25, radius=0.75, shapeName=dyn_front_l_01_capsule, center=<-1 0 0>, rotation=<rot 90 0 0>}]
 dyn_front_r_02 [{shapeType=Capsule, length=10, radius=2.5, shapeName=dyn_front_l_02_capsule, center=<-1 0 0>, rotation=<rot 90 0 0>}]
 dyn_front_r_03 [{shapeType=Capsule, length=9.75, radius=2.5, shapeName=dyn_front_l_03_capsule, center=<-1 0 0>, rotation=<rot 90 0 0>}]
 dyn_front_r_04 [{shapeType=Capsule, length=9.5, radius=2.5, shapeName=dyn_front_l_04_capsule, center=<-1 0 0>, rotation=<rot 90 0 0>}]
 dyn_front_r_05 [{shapeType=Capsule, length=7, radius=3, shapeName=dyn_back_l_05_capsule, center=<-1 0 0>, rotation=<rot 90 0 0>}]
 dyn_mstr_back_l [{shapeType=Capsule, length=3, radius=1, shapeName=dyn_mstr_front_l_capsule}]
 dyn_mstr_back_r [{shapeType=Capsule, length=3, radius=1, shapeName=dyn_mstr_front_l_capsule}]
 dyn_mstr_front_l [{shapeType=Capsule, length=3, radius=1, shapeName=dyn_mstr_front_l_capsule}]
 dyn_mstr_front_r [{shapeType=Capsule, length=3, radius=1, shapeName=dyn_mstr_front_l_capsule}]
 pelvis [{shapeType=Capsule, length=37.2982, radius=7.9957, shapeName=pelvis_capsule, center=<-15.9907 -0.7648 -0.1505>, rotation=<rot 90 0 0>}]
 spine_01 [{shapeType=Capsule, length=7.209, radius=9.9729, shapeName=spine_01_capsule, center=<3.4989 0.5274 0>, rotation=<rot 90 0 0>}]
 thigh_l [
  {shapeType=Sphere, radius=10.5, shapeName=thigh_l_sphere, center=<1.441 -1.1821 1.1289>}
  {shapeType=Capsule, length=30, radius=8, shapeName=thigh_l_capsule, center=<-19.1583 0 0>, rotation=<rot 90 0 0>}
 ]
 thigh_r [
  {shapeType=Sphere, radius=10.5, shapeName=thigh_l_sphere, center=<-1.441 1.1821 -1.1289>}
  {shapeType=Capsule, length=30, radius=8, shapeName=thigh_l_capsule, center=<19.1583 0 0>, rotation=<rot 90 0 0>}
 ]
}
constraints [
 {bone1Name=dyn_front_l_01, bone2Name=dyn_mstr_front_l, swing1LimitDegrees=55, swing2LimitDegrees=10, twistLimitDegrees=10}
 {bone1Name=dyn_front_l_02, bone2Name=dyn_front_l_01, swing1LimitDegrees=20, swing2LimitDegrees=40, twistLimitDegrees=10}
 {bone1Name=dyn_front_l_03, bone2Name=dyn_front_l_02, swing1LimitDegrees=20, swing2LimitDegrees=30, twistLimitDegrees=5}
 {bone1Name=dyn_front_l_04, bone2Name=dyn_front_l_03, swing1LimitDegrees=20, swing2LimitDegrees=30, twistLimitDegrees=5}
 {bone1Name=dyn_back_l_01, bone2Name=dyn_mstr_back_l, swing1LimitDegrees=40, swing2LimitDegrees=3, twistLimitDegrees=3}
 {bone1Name=dyn_back_l_02, bone2Name=dyn_back_l_01, swing1LimitDegrees=20, swing2LimitDegrees=20, twistLimitDegrees=5}
 {bone1Name=dyn_back_l_03, bone2Name=dyn_back_l_02, swing1LimitDegrees=20, swing2LimitDegrees=15, twistLimitDegrees=5}
 {bone1Name=dyn_back_l_04, bone2Name=dyn_back_l_03, swing1LimitDegrees=20, swing2LimitDegrees=15, twistLimitDegrees=5}
 {bone1Name=dyn_back_l_05, bone2Name=dyn_back_l_04, swing1LimitDegrees=30, swing2LimitDegrees=20, twistLimitDegrees=5}
 {bone1Name=dyn_back_l_06, bone2Name=dyn_back_l_05, swing1LimitDegrees=30, swing2LimitDegrees=20, twistLimitDegrees=5}
 {
  bone1Name=thigh_l
  bone2Name=pelvis
  swing1LimitDegrees=45
  swing1Motion=Free
  swing2LimitDegrees=45
  swing2Motion=Free
  twistLimitDegrees=45
  twistMotion=Free
 }
 {
  bone1Name=dyn_mstr_front_l
  bone2Name=pelvis
  swing1LimitDegrees=45
  swing1Motion=Free
  swing2LimitDegrees=45
  swing2Motion=Free
  twistLimitDegrees=45
  twistMotion=Free
 }
 {bone1Name=dyn_front_l_05, bone2Name=dyn_front_l_04, swing1LimitDegrees=20, swing2LimitDegrees=20, twistLimitDegrees=3}
 {
  bone1Name=dyn_front_l_05
  bone2Name=dyn_back_l_06
  swing1LimitDegrees=45
  swing1Motion=Free
  swing2LimitDegrees=45
  swing2Motion=Free
  twistLimitDegrees=45
  twistMotion=Free
 }
 {
  bone1Name=dyn_front_l_05
  bone2Name=dyn_front_l_04
  swing1LimitDegrees=45
  swing1Motion=Free
  swing2LimitDegrees=45
  swing2Motion=Free
  twistLimitDegrees=45
  twistMotion=Free
 }
 {
  bone1Name=spine_01
  bone2Name=pelvis
  swing1LimitDegrees=45
  swing1Motion=Free
  swing2LimitDegrees=45
  swing2Motion=Free
  twistLimitDegrees=45
  twistMotion=Free
 }
 {
  bone1Name=dyn_mstr_back_r
  bone2Name=pelvis
  swing1LimitDegrees=45
  swing1Motion=Free
  swing2LimitDegrees=45
  swing2Motion=Free
  twistLimitDegrees=45
  twistMotion=Free
 }
 {bone1Name=dyn_back_r_01, bone2Name=dyn_mstr_back_r, swing1LimitDegrees=40, swing2LimitDegrees=3, twistLimitDegrees=3}
 {bone1Name=dyn_back_r_02, bone2Name=dyn_back_r_01, swing1LimitDegrees=20, swing2LimitDegrees=20, twistLimitDegrees=5}
 {bone1Name=dyn_back_r_03, bone2Name=dyn_back_r_02, swing1LimitDegrees=20, swing2LimitDegrees=20, twistLimitDegrees=5}
 {bone1Name=dyn_back_r_04, bone2Name=dyn_back_r_03, swing1LimitDegrees=20, swing2LimitDegrees=15, twistLimitDegrees=5}
 {bone1Name=dyn_back_r_05, bone2Name=dyn_back_r_04, swing1LimitDegrees=30, swing2LimitDegrees=20, twistLimitDegrees=5}
 {bone1Name=dyn_back_r_06, bone2Name=dyn_back_r_05, swing1LimitDegrees=30, swing2LimitDegrees=20, twistLimitDegrees=5}
 {
  bone1Name=dyn_mstr_back_l
  bone2Name=pelvis
  swing1LimitDegrees=45
  swing1Motion=Free
  swing2LimitDegrees=45
  swing2Motion=Free
  twistLimitDegrees=45
  twistMotion=Free
 }
 {
  bone1Name=dyn_mstr_front_r
  bone2Name=pelvis
  swing1LimitDegrees=45
  swing1Motion=Free
  swing2LimitDegrees=45
  swing2Motion=Free
  twistLimitDegrees=45
  twistMotion=Free
 }
 {bone1Name=dyn_front_r_01, bone2Name=dyn_mstr_front_r, swing1LimitDegrees=55, swing2LimitDegrees=10, twistLimitDegrees=10}
 {bone1Name=dyn_front_r_02, bone2Name=dyn_front_r_01, swing1LimitDegrees=20, swing2LimitDegrees=20, twistLimitDegrees=10}
 {bone1Name=dyn_front_r_03, bone2Name=dyn_front_r_02, swing1LimitDegrees=40, swing2LimitDegrees=40, twistLimitDegrees=5}
 {bone1Name=dyn_front_r_04, bone2Name=dyn_front_r_03, swing1LimitDegrees=20, swing2LimitDegrees=20, twistLimitDegrees=5}
 {bone1Name=dyn_front_r_05, bone2Name=dyn_front_r_04, swing1LimitDegrees=30, swing2LimitDegrees=20, twistLimitDegrees=3}
 {
  bone1Name=dyn_front_r_05
  bone2Name=dyn_front_r_04
  swing1LimitDegrees=45
  swing1Motion=Free
  swing2LimitDegrees=45
  swing2Motion=Free
  twistLimitDegrees=45
  twistMotion=Free
 }
 {
  bone1Name=dyn_front_r_05
  bone2Name=dyn_back_r_06
  swing1LimitDegrees=45
  swing1Motion=Free
  swing2LimitDegrees=45
  swing2Motion=Free
  twistLimitDegrees=45
  twistMotion=Free
 }
 {
  bone1Name=thigh_r
  bone2Name=pelvis
  swing1LimitDegrees=45
  swing1Motion=Free
  swing2LimitDegrees=45
  swing2Motion=Free
  twistLimitDegrees=45
  twistMotion=Free
 }
]
