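# Post-copy AnimBlueprint for the armored Dark Storm medium male body, adding crouch correction and secondary physics to the parent mesh's pose.
# The copied pose runs through a crouch-aware Control Rig, additive offsets on three left pouch bones, then pelvis-space RigidBody with overridable gravity.
# Crouch events snap the offset alpha and reset dynamics.
# Updates also compute thigh-rotation values, which this AnimGraph never reads; a second RigidBody node is unconnected.
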AnimBlueprint /BRCosmetics/Characters/Player/Male/Medium/Bodies/M_MED_Dark_Storm_Armored/Meshes/M_MED_Dark_Storm_Armored_AnimBP
target_skeleton=&/M_MED_Dark_Storm_Armored_Skeleton
graphs [
 AnimGraph {nodes [
   RigidBody() {
    bClampLinearTranslationLimitToRefPose=true
    overridePhysicsAsset=/BRCosmetics/Characters/Player/Male/Medium/Bodies/M_MED_Dark_Storm/Meshes/M_MED_Dark_Storm_Physics
    simulationSpace=BaseBoneSpace
    alphaBoolBlend {blendOption=Linear}
    baseBoneRef {boneName=pelvis}
    componentAppliedLinearAccClamp=<3000 3000 3000>
    componentLinearAccScale=<0.2 0.2 0.2>
    componentLinearVelScale=<0.2 0.1 0.1>
   }
   n1 = Control Rig!(Source=Copy Pose From Mesh() {bUseAttachedParent=true}, Alpha, IsCrouchCtrl:bool=Get Crouching() -> (Crouching:bool)) -> (Pose) {
    alphaBoolBlend {blendOption=Linear}
    controlRigAssetReference {blueprintRigClass=&/M_MED_Dark_Storm_Armored_CtrlRig.M_MED_Dark_Storm_Armored_CtrlRig_C}
   }
   n2 = Get crouch alpha() -> (crouch alpha:double)
   n3 = "Transform (Modify) Bone - Bone: dyn_mstr_pouch_01_l"(
    ComponentPose=Local To Component(LocalPose=@n1)
    Alpha=@n2
    Translation="-0.971,-4.879,1.128"
    Rotation="-0.054,-21.316,0.007"
    Scale="1,1,1"
   ) {
    rotationMode=BMM_Additive
    rotationSpace=BCS_BoneSpace
    translationMode=BMM_Additive
    translationSpace=BCS_BoneSpace
    alphaBoolBlend {blendOption=Linear}
    alphaScaleBiasClamp {bInterpResult=true}
    boneToModify {boneName=dyn_mstr_pouch_01_l}
    rotation=<rot -0.054 -21.316 0.007>
    translation=<-0.971 -4.879 1.128>
   }
   n4 = "Transform (Modify) Bone - Bone: dyn_mstr_pouch_02_l"(ComponentPose=@n3, Alpha=@n2, Translation="0.119,-4.825,1.128", Rotation="-0.054,-21.316,0.007", Scale="1,1,1") {
    rotationMode=BMM_Additive
    rotationSpace=BCS_BoneSpace
    translationMode=BMM_Additive
    translationSpace=BCS_BoneSpace
    alphaBoolBlend {blendOption=Linear}
    alphaScaleBiasClamp {bInterpResult=true}
    boneToModify {boneName=dyn_mstr_pouch_02_l}
    rotation=<rot -0.054 -21.316 0.007>
    translation=<0.119 -4.825 1.128>
   }
   n5 = "Transform (Modify) Bone - Bone: dyn_mstr_pouch_03_l"(ComponentPose=@n4, Alpha=@n2, Translation="0.208,-4.411,0.991", Rotation="-0.054,-21.316,0.007", Scale="1,1,1") {
    rotationMode=BMM_Additive
    rotationSpace=BCS_BoneSpace
    translationMode=BMM_Additive
    translationSpace=BCS_BoneSpace
    alphaBoolBlend {blendOption=Linear}
    alphaScaleBiasClamp {bInterpResult=true}
    boneToModify {boneName=dyn_mstr_pouch_03_l}
    rotation=<rot -0.054 -21.316 0.007>
    translation=<0.208 -4.411 0.991>
   }
   n6 = RigidBody!(ComponentPose=@n5, Alpha, OverrideWorldGravity:Vector=Get GravityOverride(), ExternalForce) -> (Pose) {
    bClampLinearTranslationLimitToRefPose=true
    bOverrideWorldGravity=true
    overridePhysicsAsset=&/M_MED_Dark_Storm_Armored_Physics
    simulationSpace=BaseBoneSpace
    alphaBoolBlend {blendOption=Linear}
    baseBoneRef {boneName=pelvis}
    componentAppliedLinearAccClamp=<3000 3000 3000>
    componentLinearAccScale=<0.2 0.2 0.3>
    componentLinearVelScale=<0.2 0.2 0.3>
    overrideWorldGravity {z=-980}
   }
   Output Pose(Result=Component To Local(ComponentPose=@n6)) {layerGroup=None}
  ]}
 EventGraph {nodes [
   Event BlueprintUpdateAnimation()
   > n1 = Sequence!(execute) -> (then_0, then_1, then_2:Exec)
   > n2 = Branch(Condition=OR Boolean!(A=Get bIsJumping(), B=Get bIsFalling(), C:bool=Get bIsCrouching()) -> (ReturnValue))
   > Set thigh_rot_r(thigh_rot_r=0)
   > Set thigh_rot_l(thigh_rot_l=0)
   Set Thigh L Rot Crouch(execute=@n2.else, Thigh L Rot Crouch:double=0) -> (Output_Get:double)
   > Set Thigh_R_Rot_Crouch(Thigh_R_Rot_Crouch:double=0) -> (Output_Get:double)
   n3 = Get Skel_Mesh()
   > Set thigh_rot_r(thigh_rot_r=GetAlphaFromReferenceTransformDelta(SkeletalMeshComponent=@n3, BoneName=thigh_r, BaseName=pelvis, InputRange=(X=0,Y=-90), TwistAxis="0,0,1"))
   > Set thigh_rot_l(thigh_rot_l=GetAlphaFromReferenceTransformDelta(SkeletalMeshComponent=@n3, BoneName=thigh_l, BaseName=pelvis, InputRange=(X=-20,Y=-101), TwistAxis="0,0,1"))
   n4 = BreakTransform(InTransform=GetDeltaTransformFromRefPose(self=GetOwnerSkeletalMeshComponent(), BoneName=thigh_l, BaseName=pelvis))
   > Set Thigh L Rot Crouch(Thigh L Rot Crouch=MapRangeClamped(Value=BreakRotator(InRot=@n4.Rotation).Yaw, InRangeA=-40, InRangeB=-100))
   n5 = BreakTransform(InTransform=GetDeltaTransformFromRefPose(self=GetOwnerSkeletalMeshComponent(), BoneName=thigh_r, BaseName=pelvis))
   > Set Thigh_R_Rot_Crouch(Thigh_R_Rot_Crouch=MapRangeClamped(Value=BreakRotator(InRot=@n5.Rotation).Yaw, InRangeA=-40, InRangeB=-100))
   n6 = Branch(execute=@n1.then_1, Condition=Get bIsCrouching())
   > Set Crouching(Crouching:bool=true) -> (Output_Get:bool)
   Set Crouching(execute=@n6.else, Crouching=false)
   n7 = Gravity Override State Function DEPRECATED(
    execute=@n1.then_2
    CharacterPartAnimInstance=Self-Reference()
    GameTimeInSeconds=GetGameTimeInSeconds()
    GravityOverrideParams=Get Gravity Override Params()
   )
   > Set GravityOverride(GravityOverride=@n7)
   Event BlueprintInitializeAnimation()
   > Set Skel_Mesh(Skel_Mesh=GetOwnerSkeletalMeshComponent())
   Event OnBeginCrouching()
   > n8 = Set crouch alpha(crouch alpha:double=1) -> (Output_Get:double)
   ResetDynamics(execute=[@n8.then, @n9.then], InTeleportType=ResetPhysics)
   > ResetDynamics(InTeleportType=TeleportPhysics)
   Event OnEndCrouching()
   > n9 = Set crouch alpha(crouch alpha=0)
  ]}
]
variables [
 crouch alpha {type=number}
 Crouching {type=boolean}
 Gravity Override Params {type=GravityOverrideParamsStruct}
 GravityOverride {type=Vector}
 Skel_Mesh {type=SkeletalMeshComponent}
 Thigh L Rot Crouch {type=number}
 Thigh_R_Rot_Crouch {type=number}
 thigh_rot_l {type=number}
 thigh_rot_r {type=number}
]
